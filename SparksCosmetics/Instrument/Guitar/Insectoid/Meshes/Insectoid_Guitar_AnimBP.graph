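# Animation Blueprint for the Insectoid guitar that gives its eye random darting movement and animates its arms.
# On update, a random-length Delay gates EyeDartRandomRotator, which builds randomly signed roll and pitch angles within a range and stores them in Eye_Rotator.
# The AnimGraph adds that rotation to the eye bone in bone space, then layers the looping idle sequence over the left and right arm_01 branches.
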AnimBlueprint /SparksCosmetics/Instrument/Guitar/Insectoid/Meshes/Insectoid_Guitar_AnimBP
target_skeleton=&/Insectoid_Guitar_Skeleton
graphs [
 AnimGraph {nodes [
   n1 = "Transform (Modify) Bone - Bone: eye"!(ComponentPose=Local To Component(LocalPose=Input Pose()), Alpha=1, Rotation=Get Eye_Rotator() -> (Eye_Rotator:Rotator)) -> (Pose) {rotationMode=BMM_Additive, rotationSpace=BCS_BoneSpace, alphaBoolBlend {blendOption=Linear}, boneToModify {boneName=eye}}
   Output Pose(Result=Layered blend per bone(
     BasePose=Component To Local(ComponentPose=@n1)
     BlendPoses_0=Sequence Player 'Insectoid_Guitar_Idle_Loop'!(bLoopAnimation:bool=True) -> (Pose) {bLoopAnimation=false, sequence=&/Insectoid_Guitar_Idle_Loop}
    ) {layerSetup [{branchFilters [{boneName=arm_01_l}, {boneName=arm_01_r}]}]}) {layerGroup=None}
  ]}
 EventGraph {nodes [
   Event BlueprintUpdateAnimation()
   > Delay(Duration=RandomFloatInRange(Min=0.33, Max=1.25))
   > n1 = Flip Flop!(None:Exec) -> (A:Exec, B:Exec, IsA:bool=~) {is=title K2Node_MacroInstance}
   n2 = EyeDartRandomRotator!(
    execute=[@n1.A, @n1.B]
    self
    MinRotation:double=5
    MaxRotation:double=15
    "ApplyToX(Roll)":bool=true
    "ApplyToY(Pitch)":bool=true
    "ApplyToZ(Yaw)":bool=false
   ) -> (then, Rotator:Rotator)
   > Set Eye_Rotator(Eye_Rotator:Rotator=@n2) -> (Output_Get:Rotator)
  ]}
 EyeDartRandomRotator {nodes [
   n1 = EyeDartRandomRotator!() -> (then, MinRotation:double=~, MaxRotation:double=~, "ApplyToX(Roll)":bool=~, "ApplyToY(Pitch)":bool=~, "ApplyToZ(Yaw)":bool=~) {is=title K2Node_FunctionEntry}
   > n2 = Sequence!(execute) -> (then_0, then_1, then_2:Exec, then_3:Exec, then_4:Exec)
   > Set Min_Rotation(Min_Rotation:double=@n1) -> (Output_Get:double)
   > Set Max_Rotation(Max_Rotation:double=@n1.MaxRotation) -> (Output_Get:double)
   Branch(execute=@n2.then_1, Condition=@n1."ApplyToX(Roll)")
   > n3 = RandomFloatGivenMinMax!(execute, self, Min:double=Get Min_Rotation() -> (Min_Rotation:double), Max:double=Get Max_Rotation() -> (Max_Rotation:double)) -> (then, OutFloat:double)
   > n4 = RandomSign!(execute, self, InFloat:double=@n3) -> (then, ReturnValue:double)
   > Set XRotation(XRotation:double=@n4) -> (Output_Get:double)
   Branch(execute=@n2.then_2, Condition=@n1."ApplyToY(Pitch)")
   > n5 = RandomFloatGivenMinMax(Min=Get Min_Rotation(), Max=Get Max_Rotation())
   > n6 = RandomSign(InFloat=@n5)
   > Set YRotation(YRotation:double=@n6) -> (Output_Get:double)
   Branch(execute=@n2.then_3, Condition=@n1."ApplyToZ(Yaw)")
   > n7 = RandomFloatGivenMinMax(Min=Get Min_Rotation(), Max=Get Max_Rotation())
   > n8 = RandomSign(InFloat=@n7)
   > Set ZRotation(ZRotation:double=@n8) -> (Output_Get:double)
   Return Node!(
    execute=@n2.then_4
    Rotator:Rotator=MakeRotator(Roll=Get XRotation() -> (XRotation:double), Pitch=Get YRotation() -> (YRotation:double), Yaw=Get ZRotation() -> (ZRotation:double))
   )
  ]}
 RandomFloatGivenMinMax {nodes [
   n1 = RandomFloatGivenMinMax!() -> (then, Min:double=~, Max:double=~) {is=title K2Node_FunctionEntry}
   > Return Node!(execute, OutFloat:double=RandomFloatInRange(Min=@n1, Max=@n1.Max))
  ]}
 RandomSign {nodes [
   n1 = RandomSign!() -> (then, InFloat:double=~) {is=title K2Node_FunctionEntry}
   > n2 = Branch(Condition=RandomBool!() -> (ReturnValue:bool))
   > n3 = Set OutFloat(OutFloat:double=@n1) -> (Output_Get:double)
   Return Node!(execute=[@n4.then, @n3.then], ReturnValue:double=Get OutFloat() -> (OutFloat:double))
   n4 = Set OutFloat(execute=@n2.else, OutFloat=float * float(A=@n1, B=-1) {is=title K2Node_PromotableOperator})
  ]}
]
variables [Eye_Rotator {type=Rotator}]
